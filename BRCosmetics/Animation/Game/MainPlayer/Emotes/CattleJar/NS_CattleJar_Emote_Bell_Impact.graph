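# Niagara impact effect for the bell in the main player's CattleJar emote, built from four emitters.
# Shockwave and SoundRipples are expanding, fading mesh rings or shells, one burst-spawned and the other also spawning at a tapering rate.
# Streaks bursts short-lived SubUV sprites outward from a sphere.
# Confetti bursts orange-yellow sprites that drift under drag and randomized gravity, shrinking away while aligned to their mesh orientation.
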NiagaraSystem /BRCosmetics/Animation/Game/MainPlayer/Emotes/CattleJar/NS_CattleJar_Emote_Bell_Impact
dependencies {
 usedDataInterfaces [&NiagaraDataInterfaceCurve, &NiagaraDataInterfaceColorCurve, &NiagaraDataInterfaceArrayColor]
 usedDynamicInputs []
 usedModules [&SystemState]
}
emitters {
 Confetti {
  emitterUpdateScript [
   &EmitterState()
   &SpawnBurst_Instantaneous(Spawn Count=40, Use Spawn Probability=false)
  ]
  particleSpawnScript [
   &InitializeParticle(
    Color=&SelectLinearColorFromArray(
     Array Sampling Mode=::NewEnumerator0
     Direct Array Index=?
     Color Selection Array=DataInterface() {propertyValues {ColorData [<rgba 1 0.1464 0 1>, <rgba 1 0.7666 0 1>]}}
     Random Seed=0
    )
    Lifetime=&RandomRangeFloat(Minimum=0.8, Maximum=2.7)
    Mass Mode=::NewEnumerator3
    Sprite Rotation Mode=::NewEnumerator2
    Sprite Size Mode=::NewEnumerator1
    Uniform Sprite Size Max=8
    Uniform Sprite Size Min=3
    UsePositionOffset=false
    Write Material Random=true
   )
   &ShapeLocation(MaskSpawnGroup=false, Override Seed=false, Sphere Radius=33, UseOwnerScale=true)
   &AddVelocity(
    Module.Use Falloff Distance=false
    Origin Offset=<0 0 -2>
    Override Seed=false
    Velocity Mode=::NewEnumerator1
    Velocity Speed=&RandomRangeFloat(Minimum=150, Maximum=330)
   ) {module=AddVelocity001}
   &InitialMeshOrientation(Use Rotation Vector=false)
  ]
  particleUpdateScript [
   &ParticleState()
   &Drag(
    Drag=&Multiply_Float(
     A=0.5
     B=&FloatFromCurve(FloatCurve=DataInterface() {propertyValues {Curve {keys [{interpMode=RCIM_Cubic, tangentMode=RCTM_User, time=0, value=0}, {interpMode=RCIM_Cubic, tangentMode=RCTM_User, time=0.2, value=1}]}}})
    )
    Use Linear Drag=true
    Use Rotational Drag=false
   )
   &GravityForce(Gravity=/Niagara/DynamicInputs/Math/LerpMultipleVectors(
     Alpha:NiagaraFloat=&Divide_Float(
      A=&MakeFloatFromInt(Int=3)
      B=&MakeFloatFromInt(Int=$Particles.SpawnGroup)
     )
     V 1:Vector3f=&RandomRangeVector(Minimum=<0 0 -360>, Maximum=<1 1 -200>)
     V 2:Vector3f=&RandomRangeVector(Minimum=<0 0 -360>, Maximum=<1 1 -200>)
     V 3:Vector3f=&RandomRangeVector(Minimum=<0 0 -100>, Maximum=<1 1 -300>)
     V 4:Vector3f=?
     V 5:Vector3f=?
     Vector Count:ENiagara_MultipleLerpCount=::NewEnumerator0
    ))
   &AerodynamicDrag(Aerodynamic Drag=0.5, Aerodynamic Rotational Drag=0, Use Linear Drag=true, Use Rotational Drag=true)
   &SolveForcesAndVelocity(Clamp Velocity=false, Limit Acceleration=false)
   &ScaleSpriteSize(Scale Factor=&Vector2DFromFloat(Value=&FloatFromCurve(
      FloatCurve=DataInterface() {propertyValues {Curve {keys [
          {arriveTangent=-0.4596, interpMode=RCIM_Cubic, leaveTangent=-0.4596, tangentMode=RCTM_User, time=0.55, value=1}
          {arriveTangent=-4.9031, interpMode=RCIM_Cubic, leaveTangent=-4.9031, tangentMode=RCTM_User, time=1, value=0}
         ]}}}
      CurveIndex=&Divide_Float(
       A=$Particles.NormalizedAge
       B=&RandomRangeFloat(Minimum=0.2, Maximum=1, Randomness Mode=::NewEnumerator2, Random Seed=?)
      )
     )))
   &ScaleColor(ScaleA=false, ScaleRGB=true)
   &AlignSpriteToMeshOrientation(SetAlignment=true, SetFacing=true)
  ]
  rendererClasses [&NiagaraSpriteRendererProperties]
  renderers [{rendererClass=&NiagaraSpriteRendererProperties}]
 }
 Shockwave {
  emitterUpdateScript [
   &EmitterState()
   &SpawnBurst_Instantaneous(Use Spawn Probability=false, UseLoopCountLimit=false)
  ]
  particleSpawnScript [
   &InitializeParticle(Color=<rgba 1 1 1 0.5>, Lifetime=0.8, Mesh Scale Mode=::NewEnumerator3, UsePositionOffset=false, Write Material Random=true)
   &InitialMeshOrientation(Mesh Orientation Mode=::NewEnumerator7, Rotation=<0 0.25 0>, Use Rotation Vector=true)
  ]
  particleUpdateScript [
   &ParticleState()
   &ScaleMeshSize(Scale Factor=&VectorFromFloat(Value=&FloatFromCurve(FloatCurve=DataInterface() {propertyValues {Curve {keys [
          {arriveTangent=3.8033, interpMode=RCIM_Cubic, leaveTangent=3.8033, tangentMode=RCTM_User, time=0, value=0.4}
          {arriveTangent=0.338, interpMode=RCIM_Cubic, leaveTangent=0.338, tangentMode=RCTM_User, time=0.3, value=0.9}
          <at 1 1 RCIM_Cubic>
         ]}}})))
   &CameraOffset(Camera Offset Amount=50)
   &DynamicMaterialParameters(
    Index 0 Param 1=2
    Index 0 Param 2=1
    Index 0 Param 3=1
    Index 0 Param 4=&FloatFromCurve(FloatCurve=DataInterface() {propertyValues {Curve {keys [
         {arriveTangent=-7.591, interpMode=RCIM_Cubic, leaveTangent=-7.591, tangentMode=RCTM_User, time=0, value=1}
         {arriveTangent=-0.7152, interpMode=RCIM_Cubic, leaveTangent=-0.7152, tangentMode=RCTM_User, time=0.2, value=0.2}
         <at 1 0 RCIM_Cubic>
        ]}}})
    Module.Index0Param0Write=true
    Module.Index0Param1Write=true
    Module.Index0Param2Write=true
    Module.Index0Param3Write=true
   )
   &ScaleColor(
    Scale Alpha=&FloatFromCurve(FloatCurve=DataInterface() {propertyValues {Curve {keys [<at 0 1>,<at 1 0>]}}}, Scale Curve=0.45)
    Scale RGB=&MakeVectorFromLinearColorRGB(Color=&ColorFromCurve(ColorCurve=DataInterface() {propertyValues {
        AlphaCurve {keys [<at 0 1 RCIM_Cubic>]}
        BlueCurve {keys [<at 0.0076 0.6>, {arriveTangent=-0.2904, interpMode=RCIM_Cubic, leaveTangent=-0.2904, time=0.0968, value=0.1}, <at 0.352 0.5 RCIM_Cubic>]}
        GreenCurve {keys [<at 0.0076 1.8>, {arriveTangent=-3.7748, interpMode=RCIM_Cubic, leaveTangent=-3.7748, time=0.0968, value=0.475}, <at 0.352 0.5 RCIM_Cubic>]}
        RedCurve {keys [<at 0.0076 3>, {arriveTangent=-7.2592, interpMode=RCIM_Cubic, leaveTangent=-7.2592, time=0.0968, value=1}, <at 0.352 0.5 RCIM_Cubic>]}
       }}))
    ScaleA=true
    ScaleRGB=true
   )
  ]
  rendererClasses [&NiagaraMeshRendererProperties]
  renderers [{rendererClass=&NiagaraMeshRendererProperties, rendererIndex=0}]
 }
 SoundRipples {
  emitterUpdateScript [
   &EmitterState()
   &SpawnBurst_Instantaneous(Use Spawn Probability=false, UseLoopCountLimit=false)
   &SpawnRate(
    SpawnRate=&FloatFromCurve(
     FloatCurve=DataInterface() {propertyValues {Curve {keys [
         {arriveTangent=-2.9782, interpMode=RCIM_Cubic, leaveTangent=-2.9782, tangentMode=RCTM_User, time=0, value=5}
         {arriveTangent=-1.035, interpMode=RCIM_Cubic, leaveTangent=-1.035, tangentMode=RCTM_User, time=1, value=3}
        ]}}}
     CurveIndex=$Emitter.NormalizedLoopAge
    )
    Use Spawn Probability=false
   )
  ]
  particleSpawnScript [
   &InitializeParticle(
    Color Maximum=<rgba 1 0.9348 0.83 0.15>
    Color Minimum=<rgba 1 0.9347 0.9 0.1>
    Color Mode=::NewEnumerator2
    Lifetime=1.5
    Mesh Scale Mode=::NewEnumerator3
    Mesh Uniform Scale=185
    UsePositionOffset=false
    Write Material Random=true
   )
   &InitialMeshOrientation(
    Mesh Orientation Mode=::NewEnumerator1
    Rotation=&RandomRangeVector(Minimum=<0 0 0>, Maximum=<1 0 0>)
    Use Rotation Vector=true
   )
   &CameraOffset(Camera Offset Amount=20)
  ]
  particleUpdateScript [
   &ParticleState()
   &ScaleMeshSize(Scale Factor=&VectorFromFloat(Value=&FloatFromCurve(FloatCurve=DataInterface() {propertyValues {Curve {keys [{tangentMode=RCTM_User, time=0, value=0.4}, {tangentMode=RCTM_User, time=1, value=1}]}}})))
   &ScaleColor(
    Scale Alpha=&FloatFromCurve(
     FloatCurve=DataInterface() {propertyValues {Curve {keys [
         {arriveTangent=9.6668, interpMode=RCIM_Cubic, leaveTangent=9.6668, tangentMode=RCTM_User, time=0, value=0}
         <at 0.1 1 RCIM_Cubic>
         {interpMode=RCIM_Cubic, tangentMode=RCTM_User, time=1, value=0}
        ]}}}
     Scale Curve=0.6
    )
    ScaleA=true
    ScaleRGB=true
   )
   &Drag(Use Linear Drag=true, Use Rotational Drag=true)
   &SolveForcesAndVelocity(Clamp Velocity=false, Limit Acceleration=false)
   &DynamicMaterialParameters(
    Index 0 Param 1=0.5
    Index 0 Param 2=1
    Index 0 Param 3=1
    Index 0 Param 4=1
    Module.Param0WriteEnabled=true
    Module.Param1WriteEnabled=true
    Module.Param2WriteEnabled=true
    Module.Param3WriteEnabled=true
   )
  ]
  rendererClasses [&NiagaraMeshRendererProperties]
  renderers [{rendererClass=&NiagaraMeshRendererProperties, rendererIndex=0}]
 }
 Streaks {
  emitterUpdateScript [
   &EmitterState(Life Cycle Mode=::NewEnumerator1, Loop Behavior=::NewEnumerator1, Loop Delay=0.015, Loop Duration=1, UseLoopDelay=false)
   &SpawnBurst_Instantaneous(Spawn Count=9, Use Spawn Probability=false)
  ]
  particleSpawnScript [
   &InitializeParticle(
    Color Maximum=<rgba 1 0.8152 0.6 1>
    Color Minimum=<rgba 1 0.589 0.3 1>
    Color Mode=::NewEnumerator2
    Lifetime=&RandomRangeFloat(Minimum=0.18, Maximum=0.22)
    Sprite Rotation Mode=::NewEnumerator2
    Sprite Size Max=<18 100>
    Sprite Size Min=<12 80>
    Sprite Size Mode=::NewEnumerator2
    UsePositionOffset=false
    Write Material Random=true
   ) {module=InitializeParticle001}
   &SubUVAnimation(End Frame=3, SubUV Animation Mode:ENiagara_SubUVLookupMode=::NewEnumerator2)
   &ShapeLocation(MaskSpawnGroup=false, Override Seed=false, Sphere Radius=30, Sphere Surface Distribution=0.6, UseOwnerScale=true)
   &AddVelocityFromPoint(Module.Use Falloff Distance=false, Velocity Strength=&RandomRangeFloat(Minimum=150, Maximum=300))
   &CameraOffset(Camera Offset Amount=15)
  ]
  particleUpdateScript [
   &ParticleState()
   &Drag(Drag=8, Use Linear Drag=true, Use Rotational Drag=false)
   &SolveForcesAndVelocity(Clamp Velocity=false, Limit Acceleration=false)
   &DynamicMaterialParameters(
    Index 0 Param 2=1
    Index 0 Param 3=1
    Index 0 Param 4=&FloatFromCurve(FloatCurve=DataInterface() {propertyValues {Curve {keys [<at 0 1>,<at 1 0.22>]}}})
    Module.Param0WriteEnabled=false
    Module.Param1WriteEnabled=true
    Module.Param2WriteEnabled=true
    Module.Param3WriteEnabled=true
   )
   &ScaleColor(
    Scale Alpha=&FloatFromCurve(
     FloatCurve=DataInterface() {propertyValues {Curve {keys [
         {interpMode=RCIM_Cubic, tangentMode=RCTM_User, time=0, value=0}
         {arriveTangent=-1.8132, interpMode=RCIM_Cubic, leaveTangent=-1.8132, tangentMode=RCTM_User, time=0.1, value=1}
         {tangentMode=RCTM_User, time=1, value=0}
        ]}}}
     Scale Curve=&RandomRangeFloat(Minimum=0.09, Maximum=0.12)
    )
    ScaleA=true
    ScaleRGB=true
   )
  ]
  rendererClasses [&NiagaraSpriteRendererProperties]
  renderers [{rendererClass=&NiagaraSpriteRendererProperties}]
 }
}
summary {
 systemName=NS_CattleJar_Emote_Bell_Impact
 emitters [
  {emitterName=Shockwave, rendererClasses [&NiagaraMeshRendererProperties]}
  {emitterName=Streaks, rendererClasses [&NiagaraSpriteRendererProperties]}
  {emitterName=Confetti, rendererClasses [&NiagaraSpriteRendererProperties]}
  {emitterName=SoundRipples, rendererClasses [&NiagaraMeshRendererProperties]}
 ]
 userVariables [
  User.bIsFrontEnd {type=NiagaraBool}
  User.bIsFrontEndPreview {type=NiagaraBool}
 ]
}
