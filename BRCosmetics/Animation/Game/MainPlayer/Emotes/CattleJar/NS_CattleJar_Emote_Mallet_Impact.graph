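# Niagara one-shot impact effect for the mallet hit in the CattleJar character emote.
# Six independent burst emitters layer an orange flash with a light renderer, expanding ground smoke rings, a drifting puff circle, hot streaks, gravity-driven pebbles and upward air streaks.
# Velocity origins offset below the impact throw streaks and pebbles upward, while lifetime curves drive size, fade and dynamic material parameters.
# Front-end and preview user booleans are exposed, with no visible use in the modules.
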NiagaraSystem /BRCosmetics/Animation/Game/MainPlayer/Emotes/CattleJar/NS_CattleJar_Emote_Mallet_Impact
dependencies {
 usedDataInterfaces [
  &NiagaraDataInterfaceCurve
  &NiagaraDataInterfaceSpriteRendererInfo
  &NiagaraDataInterfaceVector2DCurve
  &NiagaraDataInterfaceColorCurve
 ]
 usedDynamicInputs []
 usedModules [&SystemState]
}
emitters {
 AirStreaks_upward {
  emitterUpdateScript [
   &EmitterState()
   &SpawnBurst_Instantaneous(Spawn Count=12, Spawn Time=0.03, Use Spawn Probability=false, UseLoopCountLimit=false)
  ]
  particleSpawnScript [
   /SetVariables(Particles.RandomFloat=&UniformRangedFloat(Minimum=0, Maximum=3)) {isSetParameters=true}
   &InitializeParticle(
    Color Maximum=<rgba 0.65 0.65 0.65 1>
    Color Minimum=<rgba 0.2257 0.238 0.2604 1>
    Color Mode=::NewEnumerator2
    Lifetime Max=0.23
    Lifetime Min=0.15
    Lifetime Mode=::NewEnumerator1
    Sprite Size Max=<25 100>
    Sprite Size Min=<10 70>
    Sprite Size Mode=::NewEnumerator2
    UsePositionOffset=false
    Write Material Random=true
   ) {module=InitializeParticle001}
   /SetVariables(Particles.SpriteSize=&Multiply_Vector2D_ByFloat(Float=&Multiply_Float(A=1, B=1), Vector2D=$Particles.SpriteSize)) {isSetParameters=true}
   &ShapeLocation(
    Handle Radius=15
    Large Radius=35
    MaskSpawnGroup=false
    Non Uniform Scale=&Multiply_VectorByFloat(Vector=<1 1 0>, Float=1)
    Override Seed=false
    Shape Primitive=::NewEnumerator3
    UseOwnerScale=true
   )
   &AddVelocity(Module.Use Falloff Distance=false, Origin Offset=<0 0 -500>, Override Seed=false, Velocity Mode=::NewEnumerator1)
   &ApplyInitialForces()
   &SubUVAnimation(End Frame=3, SubUV Animation Mode:ENiagara_SubUVLookupMode=::NewEnumerator2)
   &CameraOffset(Camera Offset Amount=20)
  ]
  particleUpdateScript [
   &ParticleState()
   &SolveForcesAndVelocity(Clamp Velocity=false, Limit Acceleration=false)
   &ScaleSpriteSize(Scale Factor=&Vector2DFromCurve(
     Vector2Curve=DataInterface() {propertyValues {XCurve {keys [<at 0.0036 1.6897>,<at 0.9967 0.0582>]}, YCurve {keys [<at 0 1>,<at 1.002 3.7227>]}}}
     ScaleCurve=&Multiply_Vector2D_ByFloat(Float=0.65, Vector2D=<1 1>)
    ))
   &ScaleColor(
    Scale Alpha=&FloatFromCurve(
     FloatCurve=DataInterface() {propertyValues {Curve {keys [
         {interpMode=RCIM_Cubic, leaveTangent=-1.9244, tangentMode=RCTM_Break, time=0, value=1}
         {interpMode=RCIM_Cubic, tangentMode=RCTM_Break, time=1, value=0}
        ]}}}
     Scale Curve=&RandomRangeFloat(Minimum=0.2, Maximum=0.4)
    )
    ScaleA=true
    ScaleRGB=true
   )
   &DynamicMaterialParameters(
    Index 0 Param 2=1
    Index 0 Param 3=1
    Index 0 Param 4=&FloatFromCurve(FloatCurve=DataInterface() {propertyValues {Curve {keys [<at 0.3 1>,<at 1 0.25>]}}})
    Module.Param0WriteEnabled=true
    Module.Param1WriteEnabled=true
    Module.Param2WriteEnabled=true
    Module.Param3WriteEnabled=true
   )
  ]
  rendererClasses [&NiagaraSpriteRendererProperties]
  renderers [{rendererClass=&NiagaraSpriteRendererProperties}]
 }
 Flame_Streaks {
  emitterUpdateScript [
   &EmitterState()
   &SpawnBurst_Instantaneous(
    Spawn Count=&RandomRangeInt(Minimum=6, Maximum=9)
    Spawn Time=0.01
    Use Spawn Probability=false
    UseLoopCountLimit=false
   )
  ]
  particleSpawnScript [
   &InitializeParticle(
    Lifetime=&RandomRangeFloat(Minimum=0.12, Maximum=0.14)
    Sprite Size=&Multiply_Vector2D_ByFloat(
     Float=&RandomRangeFloat(Minimum=0.7, Maximum=2)
     Vector2D=&RandomRangeVector2D(Minimum=<8 33>, Maximum=<13 40>)
    )
    Sprite Size Mode=::NewEnumerator4
    UsePositionOffset=false
    Write Material Random=true
   )
   &SubUVAnimation(
    End Frame Range Override=3
    Evaluation Type=::NewEnumerator0
    "Sprite Renderer "=DataInterface() {propertyValues {SpriteRenderer=&self:Flame_Streaks_1.NiagaraSpriteRendererProperties_0}}
    SubUV Animation Mode:ENiagara_SubUVLookupModeV2=::NewEnumerator2
    UseEndFrame=true
    UseStartFrame=true
   )
   &ShapeLocation(Handle Radius=1, Large Radius=50, MaskSpawnGroup=false, Override Seed=false, Shape Primitive=::NewEnumerator3, UseOwnerScale=true)
   &AddVelocity(
    Module.Use Falloff Distance=false
    Origin Offset=&MakeVector(X=0, Y=0, Z=&RandomRangeFloat(Minimum=-5, Maximum=-20))
    Override Seed=false
    Velocity Mode=::NewEnumerator1
    Velocity Speed=&RandomRangeFloat(Minimum=300, Maximum=400)
   )
   &CameraOffset(Camera Offset Amount=10)
  ]
  particleUpdateScript [
   &Drag(
    Drag=&Multiply_Float(
     A=&RandomRangeFloat(Minimum=10, Maximum=20)
     B=&FloatFromCurve(FloatCurve=DataInterface() {propertyValues {Curve {keys [<at 0 0>,<at 0.3 1>]}}})
    )
    Use Linear Drag=true
    Use Rotational Drag=true
   )
   &SolveForcesAndVelocity(Clamp Velocity=false, Limit Acceleration=false)
   &ParticleState()
   &ScaleSpriteSize(
    Non-Uniform Curve Sprite Scale=DataInterface() {propertyValues {
      XCurve {keys [{interpMode=RCIM_Cubic, tangentMode=RCTM_User, time=0, value=1}, {interpMode=RCIM_Cubic, tangentMode=RCTM_User, time=1, value=1}]}
      YCurve {keys [{interpMode=RCIM_Cubic, tangentMode=RCTM_User, time=0.35, value=1.3}, {interpMode=RCIM_Cubic, tangentMode=RCTM_User, time=1, value=0.7}]}
     }}
    Scale Sprite Size Mode=::NewEnumerator3
   )
   &ScaleColor(
    Scale Alpha=&FloatFromCurve(FloatCurve=DataInterface() {propertyValues {Curve {keys [<at 0 1>,<at 1 0>]}}})
    Scale RGB=&MakeVectorFromLinearColorRGB(Color=&ColorFromCurve(ColorCurve=DataInterface() {propertyValues {
        AlphaCurve {keys [<at 0 1 RCIM_Cubic>]}
        BlueCurve {keys [<at 0.0228 1.61 RCIM_Cubic>,<at 0.2447 0.0037>,<at 1 0 RCIM_Cubic>]}
        GreenCurve {keys [<at 0.0228 4.7542 RCIM_Cubic>,<at 0.2447 0.8205>,<at 1 0 RCIM_Cubic>]}
        RedCurve {keys [<at 0.0228 7 RCIM_Cubic>,<at 0.2447 3>,<at 1 1 RCIM_Cubic>]}
       }}))
    ScaleA=true
    ScaleRGB=true
   )
  ]
  rendererClasses [&NiagaraSpriteRendererProperties]
  renderers [{rendererClass=&NiagaraSpriteRendererProperties}]
 }
 Impact_Flash {
  emitterUpdateScript [
   &EmitterState()
   &SpawnBurst_Instantaneous(Use Spawn Probability=false, UseLoopCountLimit=false)
  ]
  particleSpawnScript [&InitializeParticle(
    Color=<rgba 1 0.2487 0.02 1>
    Lifetime=0.15
    Position Offset=<0 0 8>
    Sprite Size Mode=::NewEnumerator3
    Uniform Sprite Size=170
    UsePositionOffset=true
    Write Material Random=true
   )]
  particleUpdateScript [
   &CameraOffset(Camera Offset Amount=&FloatFromCurve(
     FloatCurve=DataInterface() {propertyValues {Curve {keys [
         {interpMode=RCIM_Cubic, tangentMode=RCTM_User, time=0, value=1}
         {arriveTangent=-0.5583, interpMode=RCIM_Cubic, leaveTangent=-0.5583, tangentMode=RCTM_User, time=0.25, value=0.3}
        ]}}}
     Scale Curve=30
    ))
   &ParticleState()
   &ScaleSpriteSize(Uniform Curve Sprite Scale=DataInterface() {propertyValues {Curve {keys [
        {arriveTangent=2.4123, interpMode=RCIM_Cubic, leaveTangent=2.4123, tangentMode=RCTM_User, time=0, value=0.3}
        {interpMode=RCIM_Cubic, tangentMode=RCTM_User, time=0.9, value=1}
       ]}}})
   &ScaleColor(
    Scale Alpha=&FloatFromCurve(
     FloatCurve=DataInterface() {propertyValues {Curve {keys [
         {interpMode=RCIM_Cubic, tangentMode=RCTM_User, time=0, value=0}
         {arriveTangent=-1.8484, interpMode=RCIM_Cubic, leaveTangent=-1.8484, tangentMode=RCTM_User, time=0.1, value=1}
         {interpMode=RCIM_Cubic, tangentMode=RCTM_User, time=1, value=0}
        ]}}}
     Scale Curve=0.15
    )
    Scale RGB=&MakeVectorFromLinearColorRGB(Color=&ColorFromCurve(ColorCurve=DataInterface() {propertyValues {
        AlphaCurve {keys [<at 0 1 RCIM_Cubic>]}
        BlueCurve {keys [<at 0.0598 2 RCIM_Cubic>,<at 0.5405 0.5 RCIM_Cubic>]}
        GreenCurve {keys [<at 0.0598 2 RCIM_Cubic>,<at 0.5405 0.5 RCIM_Cubic>]}
        RedCurve {keys [<at 0.0598 2 RCIM_Cubic>,<at 0.5405 1 RCIM_Cubic>]}
       }}))
    ScaleA=true
    ScaleRGB=true
   )
   /SetVariables(Particles.ColorFlare:LinearColor=&MultiplyLinearColor(Linear Color=$Particles.Color, Scale Factor=<10 10 10 10>)) {isSetParameters=true}
  ]
  rendererClasses [&NiagaraSpriteRendererProperties, &NiagaraLightRendererProperties]
  renderers [
   {rendererClass=&NiagaraSpriteRendererProperties}
   {rendererClass=&NiagaraSpriteRendererProperties, rendererIndex=1}
   {rendererClass=&NiagaraLightRendererProperties, rendererIndex=2}
  ]
 }
 Pebbles {
  emitterUpdateScript [
   &EmitterState()
   &SpawnBurst_Instantaneous(
    Spawn Count=&RandomRangeInt(Minimum=5, Maximum=8)
    Spawn Time=0.04
    Use Spawn Probability=false
    UseLoopCountLimit=false
   )
  ]
  particleSpawnScript [
   &InitializeParticle(
    Color Maximum=<rgba 0.4 0.4 0.4 1>
    Color Minimum=<rgba 0.1 0.1 0.1 1>
    Color Mode=::NewEnumerator2
    Lifetime=&RandomRangeFloat(Minimum=0.2, Maximum=0.6)
    Sprite Rotation Mode=::NewEnumerator2
    Sprite Size Mode=::NewEnumerator1
    Uniform Sprite Size Max=12
    Uniform Sprite Size Min=6
    UsePositionOffset=false
    Write Material Random=true
   )
   &SubUVAnimation(
    End Frame Range Override=3
    Evaluation Type=::NewEnumerator0
    "Sprite Renderer "=DataInterface() {propertyValues {SpriteRenderer=&self:Pebbles_0.NiagaraSpriteRendererProperties_0}}
    SubUV Animation Mode:ENiagara_SubUVLookupModeV2=::NewEnumerator2
    UseEndFrame=true
    UseStartFrame=true
   )
   &ShapeLocation(Handle Radius=10, Large Radius=40, MaskSpawnGroup=false, Override Seed=false, Shape Primitive=::NewEnumerator3, UseOwnerScale=true)
   &AddVelocity(
    Module.Use Falloff Distance=false
    Origin Offset=&MakeVector(X=0, Y=0, Z=&RandomRangeFloat(Minimum=-150, Maximum=-300))
    Override Seed=false
    Velocity Mode=::NewEnumerator1
    Velocity Speed=&RandomRangeFloat(Minimum=200, Maximum=400)
   )
   &CameraOffset(Camera Offset Amount=10)
  ]
  particleUpdateScript [
   &Drag(Drag=&RandomRangeFloat(Minimum=1, Maximum=3), Use Linear Drag=true, Use Rotational Drag=true)
   &GravityForce(Gravity=&MakeVector(X=0, Y=0, Z=&RandomRangeFloat(Minimum=-500, Maximum=-980)))
   &SolveForcesAndVelocity(Clamp Velocity=false, Limit Acceleration=false)
   &ParticleState()
   &ScaleSpriteSize(
    Scale Sprite Size Mode=::NewEnumerator0
    Uniform Scale Factor=&FloatFromCurve(FloatCurve=DataInterface() {propertyValues {Curve {keys [<at 0 1>,<at 1 0>]}}})
   )
  ]
  rendererClasses [&NiagaraSpriteRendererProperties]
  renderers [{rendererClass=&NiagaraSpriteRendererProperties}]
 }
 SmokeGroundRings {
  emitterUpdateScript [
   &EmitterState()
   &SpawnBurst_Instantaneous(Spawn Count=2, Use Spawn Probability=false, UseLoopCountLimit=false)
  ]
  particleSpawnScript [
   &PartitionParticles(
    Normalized Index In Partition Includes One=true
    Number Of Partitions=1
    Particles Per Partition=2
    Partition Mode:ENiagara_PartitionMode=::NewEnumerator1
   )
   &InitializeParticle(
    Lifetime=0.3
    Sprite Rotation Mode=::NewEnumerator2
    Sprite Size Mode=::NewEnumerator3
    Uniform Sprite Size=&Lerp_Float(A=2.5, B=1.875, Alpha=$Particles.PartitionParticles.Normalized Index In Partition)
    UsePositionOffset=false
    Write Material Random=true
   ) {module=InitializeParticle001}
   /SetVariables(Particles.SpriteAlignment=&RandomVector(Vector Scale=1), Particles.SpriteFacing=<0 0 1>) {isSetParameters=true}
  ]
  particleUpdateScript [
   &ParticleState()
   &DynamicMaterialParameters(
    Index 0 Param 1=&FloatFromCurve(FloatCurve=DataInterface() {propertyValues {Curve {keys [
         {arriveTangent=3.8304, interpMode=RCIM_Cubic, leaveTangent=3.8304, tangentMode=RCTM_User, time=0, value=0.5}
         {arriveTangent=0.0807, interpMode=RCIM_Cubic, leaveTangent=0.0807, tangentMode=RCTM_User, time=1, value=2}
        ]}}})
    Module.Param0WriteEnabled=true
    Module.Param1WriteEnabled=false
    Module.Param2WriteEnabled=false
    Module.Param3WriteEnabled=false
   )
   &ScaleSpriteSize(Scale Factor=&Vector2DFromFloat(Value=&FloatFromCurve(
      FloatCurve=DataInterface() {propertyValues {Curve {keys [
          {arriveTangent=4.2264, interpMode=RCIM_Cubic, leaveTangent=4.2264, tangentMode=RCTM_User, time=0, value=0}
          {interpMode=RCIM_Cubic, leaveTangent=0.6903, time=0.6, value=1}
         ]}}}
      Scale Curve=80
     )))
   &CameraOffset(Camera Offset Amount=20)
   &ScaleColor(
    Scale Alpha=&FloatFromCurve(FloatCurve=DataInterface() {propertyValues {Curve {keys [<at 0 1 RCIM_Cubic>, {interpMode=RCIM_Cubic, leaveTangent=-4.9981, time=1.0011, value=0.002}]}}})
    ScaleA=true
    ScaleRGB=false
   )
  ]
  rendererClasses [&NiagaraSpriteRendererProperties]
  renderers [{rendererClass=&NiagaraSpriteRendererProperties}]
 }
 SmokePuffCircle {
  emitterUpdateScript [
   &EmitterState()
   &SpawnBurst_Instantaneous(
    Spawn Count=&RandomRangeInt(Minimum=20, Maximum=30)
    Spawn Time=0.01
    Use Spawn Probability=true
    UseLoopCountLimit=false
   )
  ]
  particleSpawnScript [
   /SetVariables(Particles.SpriteRotationRate=&RandomRangeFloat(Minimum=-0.8, Maximum=0.8, Evaluation Type=::NewEnumerator1)) {isSetParameters=true}
   &InitializeParticle(
    Color=<rgba 1 1 1 0.1>
    Lifetime Max=0.7
    Lifetime Min=0.45
    Lifetime Mode=::NewEnumerator1
    Sprite Rotation Mode=::NewEnumerator2
    Sprite Size Mode=::NewEnumerator1
    Uniform Sprite Size Max=25
    Uniform Sprite Size Min=20
    UsePositionOffset=false
    Write Material Random=true
   ) {module=InitializeParticle001}
   &ShapeLocation(Handle Radius=1, Large Radius=20, MaskSpawnGroup=false, Override Seed=false, Shape Primitive=::NewEnumerator3, UseOwnerScale=true)
   &CameraOffset(Camera Offset Amount=5)
   &AddVelocity(Module.Use Falloff Distance=false, Override Seed=false, Velocity Mode=::NewEnumerator1, Velocity Speed=300)
  ]
  particleUpdateScript [
   &ParticleState()
   &SpriteRotationRate(Rotation Rate=&RandomRangeFloat(Minimum=-0.8, Maximum=0.8))
   &ScaleColor(
    Scale Alpha=&FloatFromCurve(FloatCurve=DataInterface() {propertyValues {Curve {keys [
         {arriveTangent=9.0725, interpMode=RCIM_Cubic, time=0, value=0}
         {arriveTangent=-2.5249, interpMode=RCIM_Cubic, leaveTangent=-2.5249, tangentMode=RCTM_User, time=0.1408, value=0.9847}
         {arriveTangent=-0.0405, interpMode=RCIM_Cubic, leaveTangent=-0.0405, tangentMode=RCTM_User, time=1, value=0}
        ]}}})
    ScaleA=true
    ScaleRGB=true
   )
   &ScaleSpriteSize(Scale Factor=&Vector2DFromFloat(Value=&FloatFromCurve(
      FloatCurve=DataInterface() {propertyValues {Curve {keys [
          {arriveTangent=1.9916, interpMode=RCIM_Cubic, time=0, value=0.5}
          {arriveTangent=0.8423, interpMode=RCIM_Cubic, leaveTangent=0.8423, tangentMode=RCTM_User, time=0.2, value=0.75}
          <at 1 1 RCIM_Cubic>
         ]}}}
      Scale Curve=2
     )))
   &AccelerationForce(Acceleration=<0 0 40>)
   &Drag(
    Drag=&FloatFromCurve(FloatCurve=DataInterface() {propertyValues {Curve {keys [{interpMode=RCIM_Cubic, tangentMode=RCTM_User, time=0, value=5}, {interpMode=RCIM_Cubic, tangentMode=RCTM_User, time=1, value=6}]}}})
    Use Linear Drag=true
    Use Rotational Drag=false
   )
   &SolveForcesAndVelocity(Clamp Velocity=false, Limit Acceleration=false)
   &DynamicMaterialParameters(
    Index 0 Param 1=&FloatFromCurve(FloatCurve=DataInterface() {propertyValues {Curve {keys [<at 0 1>,<at 1 0>]}}}, Scale Curve=0.2)
    Module.Param0WriteEnabled=true
    Module.Param1WriteEnabled=false
    Module.Param2WriteEnabled=false
    Module.Param3WriteEnabled=false
   )
  ]
  rendererClasses [&NiagaraSpriteRendererProperties]
  renderers [{rendererClass=&NiagaraSpriteRendererProperties}]
 }
}
summary {
 systemName=NS_CattleJar_Emote_Mallet_Impact
 emitters [
  {emitterName=SmokeGroundRings, rendererClasses [&NiagaraSpriteRendererProperties]}
  {emitterName=SmokePuffCircle, rendererClasses [&NiagaraSpriteRendererProperties]}
  {emitterName=Flame_Streaks, rendererClasses [&NiagaraSpriteRendererProperties]}
  {emitterName=Pebbles, rendererClasses [&NiagaraSpriteRendererProperties]}
  {emitterName=AirStreaks_upward, rendererClasses [&NiagaraSpriteRendererProperties]}
  {emitterName=Impact_Flash, rendererClasses [&NiagaraSpriteRendererProperties, &NiagaraLightRendererProperties]}
 ]
 userVariables [
  User.bIsFrontEnd {type=NiagaraBool}
  User.bIsFrontEndPreview {type=NiagaraBool}
 ]
}
