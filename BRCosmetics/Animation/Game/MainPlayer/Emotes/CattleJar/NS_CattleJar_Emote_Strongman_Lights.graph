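# Niagara system for the CattleJar Strongman emote that lights up a column of orange bulbs with glows.
# LightBulbs bursts a mesh particle whose alpha ramps up over its short lifetime, while LightGlow bursts five camera-offset sprites at hard-coded, vertically stacked positions.
# Each glow's alpha snaps on or off by comparing a rising curve to its normalized index, staggering the lights in sequence.
# Particles persist past their lifetime.
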NiagaraSystem /BRCosmetics/Animation/Game/MainPlayer/Emotes/CattleJar/NS_CattleJar_Emote_Strongman_Lights
dependencies {
 usedDataInterfaces [&NiagaraDataInterfaceCurve, &NiagaraDataInterfaceArrayFloat3]
 usedDynamicInputs []
 usedModules [&SystemState]
}
emitters {
 LightBulbs {
  emitterUpdateScript [
   &EmitterState(Life Cycle Mode=::NewEnumerator1, Loop Behavior=::NewEnumerator1, Loop Duration Mode=::NewEnumerator1, UseLoopDelay=false)
   &SpawnBurst_Instantaneous(Use Spawn Probability=false, UseLoopCountLimit=false)
  ]
  particleSpawnScript [
   &InitializeParticle(Color=<rgba 100 16.6667 0 1>, Lifetime=0.144, UsePositionOffset=false, Write Material Random=true)
   &CameraOffset()
  ]
  particleUpdateScript [
   &ParticleState(Kill Particles When Lifetime Has Elapsed=false, Let Infinitely Lived Particles Die When Emitter Deactivates=true)
   &ScaleColor(
    Scale Alpha=&FloatFromCurve(FloatCurve=DataInterface() {propertyValues {Curve {keys [<at 0 0>,<at 1 1>]}}})
    ScaleA=true
    ScaleRGB=true
   )
  ]
  rendererClasses [&NiagaraMeshRendererProperties]
  renderers [{rendererClass=&NiagaraMeshRendererProperties, rendererIndex=0}]
 }
 LightGlow {
  emitterUpdateScript [
   &EmitterState(Life Cycle Mode=::NewEnumerator1, Loop Behavior=::NewEnumerator1, Loop Delay=0.07, Loop Duration Mode=::NewEnumerator1, UseLoopDelay=false)
   &SpawnBurst_Instantaneous(Spawn Count=5, Use Spawn Probability=false, UseLoopCountLimit=false)
  ]
  particleSpawnScript [
   &InitializeParticle(
    Color=<rgba 3 0.3 0 1>
    Lifetime=0.1
    Sprite Size Mode=::NewEnumerator3
    Uniform Sprite Size=45
    UsePositionOffset=false
    Write Material Random=true
   )
   /SetVariables(Particles.Position=&SelectVectorFromArray(
     Direct Array Index=&MakeIntFromFloat(Float=&Multiply_Float(A=&ReturnNormalizedExecIndex(Normalized Index Includes One=false), B=5))
     Vector Selection Array=DataInterface() {propertyValues {FloatData [<40 32 35>, <36 32 68>, <33 32 95>, <30 32 119>, <28 32 142>]}}
    )) {isSetParameters=true}
   &CameraOffset(Camera Offset Amount=-20)
  ]
  particleUpdateScript [
   &ParticleState(Kill Particles When Lifetime Has Elapsed=false, Let Infinitely Lived Particles Die When Emitter Deactivates=true)
   &ScaleColor(
    Scale Alpha=&FloatFromCurve(FloatCurve=DataInterface() {propertyValues {Curve {keys [<at 0 0>,<at 0.5 1>]}}})
    ScaleA=true
    ScaleRGB=true
   ) {enabled=false}
   &ScaleColor(
    Scale Alpha=&MakeCustomFloatFromBool(
     Bool=&SetBoolByFloatComparison(
      A=&FloatFromCurve(FloatCurve=DataInterface() {propertyValues {Curve {keys [
           {arriveTangent=0.1593, interpMode=RCIM_Cubic, leaveTangent=0.1593, tangentMode=RCTM_User, time=0, value=0}
           {arriveTangent=1.3454, interpMode=RCIM_Cubic, leaveTangent=1.3454, tangentMode=RCTM_User, time=1.1, value=1}
          ]}}})
      B=&ReturnNormalizedExecIndex()
      Comparison Type=::NewEnumerator2
     )
     True Float=1
     False Float=0
    )
    ScaleA=true
    ScaleRGB=true
   ) {module=ScaleColor001}
  ]
  rendererClasses [&NiagaraSpriteRendererProperties]
  renderers [{rendererClass=&NiagaraSpriteRendererProperties}]
 }
}
summary {
 systemName=NS_CattleJar_Emote_Strongman_Lights
 emitters [
  {emitterName=LightBulbs, rendererClasses [&NiagaraMeshRendererProperties]}
  {emitterName=LightGlow, rendererClasses [&NiagaraSpriteRendererProperties]}
 ]
 userVariables [
  User.bIsFrontEnd {type=NiagaraBool}
  User.bIsFrontEndPreview {type=NiagaraBool}
 ]
}
