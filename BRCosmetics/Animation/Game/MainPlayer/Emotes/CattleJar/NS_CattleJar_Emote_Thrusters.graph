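# Niagara system for the CattleJar player emote: a timed thruster firing from three hammer exhausts.
# Emitters index a shared Position_HammerExhausts array to layer flame meshes, glow sprites, fire ribbons, smoke and spark bursts, alongside an engine glow light.
# Black smoke ribbons are placed procedurally along Y by order index, with panning, eroding material.
# Fire, smoke and spark spawns rotate exhaust offsets by owner rotation; BlackSmokeTrails001 is disabled.
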NiagaraSystem /BRCosmetics/Animation/Game/MainPlayer/Emotes/CattleJar/NS_CattleJar_Emote_Thrusters
dependencies {
 usedDataInterfaces [
  &NiagaraDataInterfaceArrayFloat3
  &NiagaraDataInterfaceCurve
  &NiagaraDataInterfaceColorCurve
  &NiagaraDataInterfaceArrayFloat
  &NiagaraDataInterfacePlatformSet
 ]
 usedDynamicInputs []
 usedModules [
  &self:SystemSpawnScript.SystemScriptSource.SystemScriptGraph.NiagaraNodeAssignment_1.SetVariables_#1
  &SystemState
 ]
}
emitters {
 BlackSmokeTrails {
  emitterSpawnScript [/SetVariables(Emitter.SmokeRandomness:NiagaraFloat=&RandomRangeFloat(Minimum=-0.2, Maximum=0.2)) {isSetParameters=true}]
  emitterUpdateScript [
   &EmitterState(Life Cycle Mode=::NewEnumerator1, Loop Behavior=::NewEnumerator1, Loop Delay=0.61, Loop Duration=1, UseLoopDelay=true)
   &SpawnBurst_Instantaneous(Spawn Count=9, Use Spawn Probability=false, UseLoopCountLimit=false)
  ]
  particleSpawnScript [
   &PartitionParticles(Number Of Partitions=9, Write Partition to Ribbon ID=::NewEnumerator0)
   &InitializeParticle(
    Color Mode=::NewEnumerator0
    Lifetime=/Niagara/DynamicInputs/Math/LerpMultipleFloats(
     Float Count:ENiagara_MultipleLerpCount=::NewEnumerator0
     1:NiagaraFloat=4
     2:NiagaraFloat=4.1
     3:NiagaraFloat=4.06
     4:NiagaraFloat=?
     5:NiagaraFloat=?
     Alpha:NiagaraFloat=&MakeFloatFromInt(Int=&ModuloInt(A=$Particles.PartitionParticles.Partition, B=3))
    )
    Ribbon Width=&RandomRangeFloat(Minimum=10, Maximum=12)
    Ribbon Width Mode=::NewEnumerator1
    Sprite Size Mode=::NewEnumerator3
    UsePositionOffset=false
    Write Material Random=true
   )
  ]
  particleUpdateScript [
   /SetVariables(Particles.ParticleRibbonOrderIndex:NiagaraFloat=&Divide_Float(
     A=&MakeFloatFromInt(Int=/Niagara/DynamicInputs/Divide/Divide_Int(A:NiagaraInt32=$Particles.PartitionParticles.Partition, B:NiagaraInt32=3))
     B=3
    )) {isSetParameters=true, module=SetVariables_#2}
   /SetVariables(Particles.SmokeLength:NiagaraFloat=&RandomRangeFloat(Minimum=80, Maximum=120)) {isSetParameters=true}
   /SetVariables(Particles.Position=&AddVectorToPosition(
     Position=&SelectVectorFromArray(
      Direct Array Index=&ModuloInt(A=$Particles.PartitionParticles.Partition, B=3)
      Vector Selection Array=$System.Position_HammerExhausts
     )
     Vector=&MakeVector(
      X=0
      Y=&Add_Float(
       A=5
       B=&Multiply_Float(
        A=&Multiply_Float(
         A=&FloatFromCurve(FloatCurve=DataInterface() {propertyValues {Curve {keys [
              {arriveTangent=1.5544, interpMode=RCIM_Cubic, leaveTangent=1.5544, tangentMode=RCTM_User, time=0, value=0.3}
              {arriveTangent=0.315, interpMode=RCIM_Cubic, leaveTangent=0.315, tangentMode=RCTM_User, time=0.4, value=0.9}
              <at 0.9 1 RCIM_Cubic>
             ]}}})
         B=&Multiply_Float(A=-1, B=$Particles.SmokeLength)
        )
        B=$Particles.ParticleRibbonOrderIndex
       )
      )
      Z=0
     )
    )) {isSetParameters=true}
   &ParticleState()
   /SetVariables(Particles.RibbonIdentifier:NiagaraInt32=&ModuloInt(A=$Particles.PartitionParticles.Partition, B=3)) {isSetParameters=true}
   /SetVariables(Particles.RibbonID=/Niagara/DynamicInputs/NiagaraID/MakeNiagaraIDFromInt(Index:NiagaraInt32=0, Acquire Tag:NiagaraInt32=$Particles.RibbonIdentifier)) {isSetParameters=true}
   /SetVariables(Particles.RibbonLinkOrder:NiagaraFloat=&Divide_Float(A=$Particles.ParticleRibbonOrderIndex, B=3)) {isSetParameters=true}
   &FN_RibbonMaterial(
    Color=&LerpLinearColors(
     StartColor=&ColorFromCurve(ColorCurve=DataInterface() {propertyValues {
        AlphaCurve {keys [<at 0 1 RCIM_Cubic>]}
        BlueCurve {keys [<at 0.0324 0.0791>, {arriveTangent=-0.2281, interpMode=RCIM_Cubic, time=0.0592, value=0.006}, <at 0.2669 0.02 RCIM_Cubic>]}
        GreenCurve {keys [<at 0.0324 0.5856>, {arriveTangent=-2.1755, interpMode=RCIM_Cubic, time=0.0592, value=0.0286}, <at 0.2669 0.022 RCIM_Cubic>]}
        RedCurve {keys [<at 0.0324 1>, {arriveTangent=-3.7443, interpMode=RCIM_Cubic, time=0.0592, value=0.2}, <at 0.2669 0.03 RCIM_Cubic>]}
       }})
     EndColor=<rgba 0.25 0.234 0.23 1>
     LerpFactor=&FloatFromCurve(
      FloatCurve=DataInterface() {propertyValues {Curve {keys [
          {interpMode=RCIM_Cubic, tangentMode=RCTM_User, time=0.05, value=0}
          {arriveTangent=0.5217, interpMode=RCIM_Cubic, leaveTangent=0.5217, tangentMode=RCTM_User, time=0.27, value=1}
         ]}}}
      CurveIndex=$Particles.ParticleRibbonOrderIndex
     )
    )
    Erosion Amount:NiagaraFloat=&ClampFloat(Float=&Add_Float(
      A=&ClampFloat(Float=&Subtract_Float(A=$Particles.ParticleRibbonOrderIndex, B=0))
      B=&FloatFromCurve(FloatCurve=DataInterface() {propertyValues {Curve {keys [
           {interpMode=RCIM_Cubic, tangentMode=RCTM_User, time=0, value=0}
           {arriveTangent=2.6369, interpMode=RCIM_Cubic, leaveTangent=2.6369, tangentMode=RCTM_User, time=1, value=1}
          ]}}})
     ))
    Erosion Power=5
    Opacity Pan U Offset:NiagaraFloat=&Add_Float(
     A=$Emitter.SmokeRandomness
     B=&SelectFloatFromArray(Direct Array Index=$Particles.RibbonIdentifier, Float Selection Array=DataInterface() {propertyValues {FloatData [0.39,0.55,0.33]}})
    )
    Opacity Pan U Speed:NiagaraFloat=&Multiply_Float(
     A=&FloatFromCurve(FloatCurve=DataInterface() {propertyValues {Curve {keys [
          {interpMode=RCIM_Cubic, tangentMode=RCTM_User, time=0.7, value=1}
          {arriveTangent=-0.5354, interpMode=RCIM_Cubic, leaveTangent=-0.5354, tangentMode=RCTM_User, time=1, value=0.9}
         ]}}})
     B=&SelectFloatFromArray(Direct Array Index=$Particles.RibbonIdentifier, Float Selection Array=DataInterface() {propertyValues {FloatData [0.13,0.1,0.17]}})
    )
    Opacity Scale U:NiagaraFloat=&SelectFloatFromArray(Direct Array Index=$Particles.RibbonIdentifier, Float Selection Array=DataInterface() {propertyValues {FloatData [0.33,0.37,0.3]}})
    Scale Alpha:NiagaraFloat=&Multiply_Float(
     A=&FloatFromCurve(
      FloatCurve=DataInterface() {propertyValues {Curve {keys [
          <at 0 0 RCIM_Cubic>
          {arriveTangent=-0.3375, interpMode=RCIM_Cubic, leaveTangent=-0.3375, tangentMode=RCTM_User, time=0.1, value=1}
          {arriveTangent=-0.3368, interpMode=RCIM_Cubic, leaveTangent=-0.3368, tangentMode=RCTM_User, time=0.8, value=0.75}
         ]}}}
      Scale Curve=2.5
     )
     B=&FloatFromCurve(
      FloatCurve=DataInterface() {propertyValues {Curve {keys [
          {arriveTangent=-1.203, interpMode=RCIM_Cubic, leaveTangent=-1.203, tangentMode=RCTM_User, time=0, value=1}
          {arriveTangent=-1.6423, interpMode=RCIM_Cubic, leaveTangent=-1.6423, tangentMode=RCTM_User, time=1, value=0.3}
         ]}}}
      CurveIndex=$Particles.ParticleRibbonOrderIndex
     )
    )
   )
   &CameraOffset(Camera Offset Amount=6)
   &ScaleRibbonWidth(Ribbon Width Scale=&Multiply_Float(
     A=&FloatFromCurve(
      FloatCurve=DataInterface() {propertyValues {Curve {keys [
          {interpMode=RCIM_Cubic, tangentMode=RCTM_User, time=0, value=1}
          {arriveTangent=2.6278, interpMode=RCIM_Cubic, leaveTangent=2.6278, tangentMode=RCTM_User, time=1, value=2.5}
         ]}}}
      CurveIndex=$Particles.ParticleRibbonOrderIndex
     )
     B=&FloatFromCurve(FloatCurve=DataInterface() {propertyValues {Curve {keys [{interpMode=RCIM_Cubic, tangentMode=RCTM_User, time=0.2, value=1}, {interpMode=RCIM_Cubic, tangentMode=RCTM_User, time=1, value=0.8}]}}})
    ))
  ]
  rendererClasses [&NiagaraRibbonRendererProperties]
  renderers [{rendererClass=&NiagaraRibbonRendererProperties}]
 }
 BlackSmokeTrails001 {
  bEnabled=false
  emitterUpdateScript [
   &EmitterState(Life Cycle Mode=::NewEnumerator1, Loop Behavior=::NewEnumerator1, Loop Delay=0.69, Loop Duration=1, UseLoopDelay=true)
   &SpawnBurst_Instantaneous(Spawn Count=9, Use Spawn Probability=false, UseLoopCountLimit=false)
   &SpawnRate(SpawnRate=30, Use Spawn Probability=false) {enabled=false}
   &SpawnBurst_Instantaneous(Spawn Count=9, Spawn Time=1, Use Spawn Probability=false, UseLoopCountLimit=false) {module=SpawnBurst_Instantaneous001}
  ]
  particleSpawnScript [
   &PartitionParticles(Number Of Partitions=3, Write Partition to Ribbon ID=::NewEnumerator0)
   &InitializeParticle(
    Lifetime=&LerpMultipleFloats(
     Float Count:ENiagara_MultipleLerpCount=::NewEnumerator0
     1:NiagaraFloat=4.3
     2:NiagaraFloat=4.5
     3:NiagaraFloat=5.7
     4:NiagaraFloat=?
     5:NiagaraFloat=?
     Alpha:NiagaraFloat=&MakeFloatFromInt(Int=$Particles.PartitionParticles.Partition)
    )
    Ribbon Width=15
    Ribbon Width Mode=::NewEnumerator1
    UsePositionOffset=false
    Write Material Random=true
   )
   &AddVelocity(Velocity=<0 0 300>) {enabled=false}
  ]
  particleUpdateScript [
   /SetVariables(Particles.ParticleRibbonOrderIndex:NiagaraFloat=&Divide_Float(
     A=&MakeFloatFromInt(Int=&MakeIntFromFloat(Float=&Add_Float(
        A=&Divide_Float(
         A=&Multiply_Float(A=&OneMinusFloat(Float=&ReturnNormalizedExecIndex()), B=10)
         B=3
        )
        B=-0.334
       )))
     B=10
    )) {enabled=false, isSetParameters=true}
   /SetVariables(Particles.ParticleRibbonOrderIndex:NiagaraFloat=&OneMinusFloat(Float=&ReturnNormalizedExecIndex())) {isSetParameters=true, module=SetVariables_#2}
   /SetVariables(Particles.Position=&AddVectorToPosition(
     Position=&SelectVectorFromArray(Direct Array Index=$Particles.PartitionParticles.Partition, Vector Selection Array=$System.Position_HammerExhausts)
     Vector=&MakeVector(
      X=0
      Y=&Add_Float(
       A=10
       B=&Multiply_Float(
        A=&Multiply_Float(
         A=&FloatFromCurve(FloatCurve=DataInterface() {propertyValues {Curve {keys [
              {arriveTangent=2.1169, interpMode=RCIM_Cubic, leaveTangent=2.1169, tangentMode=RCTM_User, time=0, value=0.4}
              {interpMode=RCIM_Cubic, tangentMode=RCTM_User, time=0.4, value=1}
             ]}}})
         B=-500
        )
        B=$Particles.ParticleRibbonOrderIndex
       )
      )
      Z=0
     )
    )) {isSetParameters=true}
   &GravityForce(Gravity=&MakeVector(
     X=0
     Y=0
     Z=&Multiply_Float(
      A=700
      B=&FloatFromCurve(
       FloatCurve=DataInterface() {propertyValues {Curve {keys [
           {interpMode=RCIM_Cubic, tangentMode=RCTM_User, time=0, value=0}
           {arriveTangent=-0.3168, interpMode=RCIM_Cubic, leaveTangent=-0.3168, tangentMode=RCTM_User, time=2, value=1}
           {interpMode=RCIM_Cubic, tangentMode=RCTM_User, time=4, value=0.2}
          ]}}}
       CurveIndex=$Emitter.Age
      )
     )
    )) {enabled=false}
   &CurlNoiseForce(
    Noise Frequency=10
    Noise Strength=&Multiply_Float(
     A=600
     B=&FloatFromCurve(FloatCurve=DataInterface() {propertyValues {Curve {keys [<at 0.4 0>,<at 0.7 1>]}}})
    )
    Pan Noise=false
   ) {enabled=false}
   &SolveForcesAndVelocity(Clamp Velocity=false, Limit Acceleration=false)
   &ParticleState()
   &ScaleColor(
    Scale Alpha=&FloatFromCurve(
     FloatCurve=DataInterface() {propertyValues {Curve {keys [
         <at 0 0 RCIM_Cubic>
         {arriveTangent=-0.0827, interpMode=RCIM_Cubic, leaveTangent=-0.0827, tangentMode=RCTM_User, time=0.2, value=1}
         {interpMode=RCIM_Cubic, tangentMode=RCTM_User, time=0.4, value=1}
         {interpMode=RCIM_Cubic, tangentMode=RCTM_User, time=1, value=0}
        ]}}}
     Scale Curve=0.75
    )
    Scale RGB=<0.015 0.011 0.01>
    ScaleA=true
    ScaleRGB=true
   ) {enabled=false}
   &ScaleRibbonWidth(Ribbon Width Scale=&FloatFromCurve(FloatCurve=DataInterface() {propertyValues {Curve {keys [<at 0 1>,<at 1 0.3>]}}})) {enabled=false}
   &DynamicMaterialParameters(
    Index 0 Param 1=&FloatFromCurve(
     FloatCurve=DataInterface() {propertyValues {Curve {keys [
         {interpMode=RCIM_Cubic, tangentMode=RCTM_User, time=0, value=0}
         {interpMode=RCIM_Cubic, tangentMode=RCTM_User, time=0.2, value=1}
         {interpMode=RCIM_Cubic, tangentMode=RCTM_User, time=0.3, value=1}
         {interpMode=RCIM_Cubic, tangentMode=RCTM_User, time=1, value=0}
        ]}}}
     Scale Curve=&RandomRangeFloat(Minimum=0.7, Maximum=1)
    )
    Index 0 Param 2=1
    Index 0 Param 3=&SelectFloatFromArray(Direct Array Index=$Particles.PartitionParticles.Partition, Float Selection Array=DataInterface() {propertyValues {FloatData [0.6,1.2,0.9]}})
    Module.Index0Param0Write=true
    Module.Index0Param1Write=false
    Module.Index0Param2Write=true
    Module.Index0Param3Write=false
   ) {enabled=false}
   &FN_RibbonMaterial(
    Color=&ColorFromCurve(
     ColorCurve=DataInterface() {propertyValues {
       AlphaCurve {keys [<at 0 1 RCIM_Cubic>]}
       BlueCurve {keys [<at 0 0.1832 RCIM_Cubic>,<at 0.0485 0.0052 RCIM_Cubic>]}
       GreenCurve {keys [<at 0 0 RCIM_Cubic>,<at 0.0485 0.0052 RCIM_Cubic>]}
       RedCurve {keys [<at 0 1 RCIM_Cubic>,<at 0.0485 0.0052 RCIM_Cubic>]}
      }}
     CurveIndex=$Particles.ParticleRibbonOrderIndex
    )
    Opacity Erosion:NiagaraFloat=&FloatFromCurve(
     FloatCurve=DataInterface() {propertyValues {Curve {keys [{interpMode=RCIM_Cubic, tangentMode=RCTM_User, time=0, value=0}, {interpMode=RCIM_Cubic, tangentMode=RCTM_User, time=0.8, value=1}]}}}
     CurveIndex=$Particles.ParticleRibbonOrderIndex
    )
    Opacity Pan U Offset:NiagaraFloat=&SelectFloatFromArray(
     Direct Array Index=$Particles.PartitionParticles.Partition
     Float Selection Array=DataInterface() {propertyValues {FloatData [0.16,0.27,0.15]}}
    )
    Opacity Pan U Speed:NiagaraFloat=&SelectFloatFromArray(
     Direct Array Index=$Particles.PartitionParticles.Partition
     Float Selection Array=DataInterface() {propertyValues {FloatData [0.16,0.6,0.45]}}
    )
    Opacity Power=0.5078
    Opacity Scale U:NiagaraFloat=&SelectFloatFromArray(
     Direct Array Index=$Particles.PartitionParticles.Partition
     Float Selection Array=DataInterface() {propertyValues {FloatData [0.17,0.16,0.19]}}
    )
    Scale Alpha:NiagaraFloat=&FloatFromCurve(FloatCurve=DataInterface() {propertyValues {Curve {keys [
         {interpMode=RCIM_Cubic, tangentMode=RCTM_User, time=0, value=0}
         {interpMode=RCIM_Cubic, tangentMode=RCTM_User, time=0.25, value=1}
         {interpMode=RCIM_Cubic, tangentMode=RCTM_User, time=1, value=0}
        ]}}})
   )
  ]
  rendererClasses [&NiagaraRibbonRendererProperties]
  renderers [
   {rendererClass=&NiagaraRibbonRendererProperties}
   {rendererClass=&NiagaraRibbonRendererProperties, rendererIndex=1}
  ]
 }
 Engine_GlowLight {
  emitterUpdateScript [
   &EmitterState()
   &SpawnBurst_Instantaneous(Use Spawn Probability=false, UseLoopCountLimit=false)
  ]
  particleSpawnScript [
   &InitializeParticle(
    Color=<rgba 1 0.2717 0.05 1>
    Lifetime=1.15
    Mesh Scale=<1.04 1.2 1.04>
    Mesh Scale Mode=::NewEnumerator4
    Sprite Size Mode=::NewEnumerator3
    Uniform Sprite Size=83
    UsePositionOffset=false
    Write Material Random=true
   )
   &SystemLocation(MaskSpawnGroup=false, Offset=<0 -35 0>)
   &CameraOffset(Camera Offset Amount=33)
   /SetVariables(Particles.CameraOffsetInnerThrusters=2, Particles.PositionInnerThrusters:NiagaraPosition=<-1 -28 1.3>) {isSetParameters=true}
   &DynamicMaterialParameters(
    Index 0 Param 1=&RandomRangeFloat(Minimum=0, Maximum=1)
    Index 0 Param 4=5
    Module.Index0Param0Write=true
    Module.Index0Param1Write=false
    Module.Index0Param2Write=false
    Module.Index0Param3Write=true
   )
  ]
  particleUpdateScript [
   &ParticleState()
   &ScaleMeshSize(Scale Factor=&MakeVector(
     X=1
     Y=&FloatFromCurve(FloatCurve=DataInterface() {propertyValues {Curve {keys [{interpMode=RCIM_Cubic, tangentMode=RCTM_User, time=0, value=0.7}, {interpMode=RCIM_Cubic, tangentMode=RCTM_User, time=0.2, value=1}]}}})
     Z=1
    ))
   &ScaleSpriteSize(Uniform Curve Sprite Scale=DataInterface() {propertyValues {Curve {keys [
        <at 0 0.3 RCIM_Cubic>
        {arriveTangent=-4.7584, interpMode=RCIM_Cubic, leaveTangent=-4.7584, tangentMode=RCTM_User, time=0.21, value=1.5}
        {interpMode=RCIM_Cubic, tangentMode=RCTM_User, time=0.5, value=1}
       ]}}})
   /SetVariables(Particles.ColorInnerThrusters:LinearColor=&MakeLinearColorFromVectorAndFloat(
     "Vector (RGB)"=&MakeVectorFromLinearColorRGB(Color=&ColorFromCurve(ColorCurve=DataInterface() {propertyValues {
         AlphaCurve {keys [<at 0 1 RCIM_Cubic>]}
         BlueCurve {keys [
           <at 0 0>
           {arriveTangent=0.812, interpMode=RCIM_Cubic, leaveTangent=0.812, time=0.0729, value=0.5127}
           <at 0.1883 0.1529>
           {interpMode=RCIM_Cubic, leaveTangent=-0.2128, time=0.7966, value=0.04}
           <at 0.8687 0>
          ]}
         GreenCurve {keys [
           <at 0 0>
           {arriveTangent=4.3099, interpMode=RCIM_Cubic, leaveTangent=4.3099, time=0.0729, value=2.705}
           <at 0.1883 0.8118>
           {interpMode=RCIM_Cubic, leaveTangent=-1.1064, time=0.7966, value=0.2621}
           <at 0.8687 0.0167>
          ]}
         RedCurve {keys [
           <at 0 0.1>
           {arriveTangent=15.397, interpMode=RCIM_Cubic, leaveTangent=15.397, time=0.0729, value=10}
           <at 0.1883 3>
           {interpMode=RCIM_Cubic, leaveTangent=-2.783, time=0.7966, value=1}
           <at 0.8687 1>
          ]}
        }}))
     "Float (Alpha)"=&FloatFromCurve(
      FloatCurve=DataInterface() {propertyValues {Curve {keys [
          {interpMode=RCIM_Cubic, tangentMode=RCTM_User, time=0, value=0}
          {interpMode=RCIM_Cubic, tangentMode=RCTM_User, time=0.2, value=1}
          {arriveTangent=-0.1377, interpMode=RCIM_Cubic, leaveTangent=-0.1377, tangentMode=RCTM_User, time=0.76, value=1}
          {arriveTangent=-4.1667, interpMode=RCIM_Cubic, leaveTangent=-4.1667, time=0.87, value=0.2}
          {arriveTangent=-0.1553, interpMode=RCIM_Cubic, leaveTangent=-0.1553, tangentMode=RCTM_User, time=1, value=0}
         ]}}}
      Scale Curve=0.7
     )
    )) {isSetParameters=true}
   /SetVariables(Particles.GlowColor=&MultiplyLinearColor(
     Linear Color=$Particles.ColorInnerThrusters
     Scale Factor=&MakeFloat4FromVectorAndFloat(
      "Vector (RGB)"=<1 1 1>
      "Float (Alpha)"=&FloatFromCurve(
       FloatCurve=DataInterface() {propertyValues {Curve {keys [{interpMode=RCIM_Cubic, tangentMode=RCTM_User, time=0.1, value=0}, {interpMode=RCIM_Cubic, tangentMode=RCTM_User, time=0.2, value=1}]}}}
       Scale Curve=0.1
      )
     )
    )) {isSetParameters=true}
   /SetVariables(Particles.LightColor=&MultiplyLinearColor(
     Linear Color=$Particles.GlowColor
     Scale Factor=&MakeFloat4FromVectorAndFloat(
      "Vector (RGB)"=<4 4 4>
      "Float (Alpha)"=&FloatFromCurve(
       FloatCurve=DataInterface() {propertyValues {Curve {keys [
           {arriveTangent=2.3626, interpMode=RCIM_Cubic, leaveTangent=2.3626, tangentMode=RCTM_User, time=0.15, value=0}
           {arriveTangent=3.3333, interpMode=RCIM_Cubic, leaveTangent=3.3333, time=0.26, value=7}
           {interpMode=RCIM_Cubic, tangentMode=RCTM_User, time=0.45, value=1}
          ]}}}
       Scale Curve=10
      )
     )
    )) {isSetParameters=true}
  ]
  rendererClasses [
   &NiagaraSpriteRendererProperties
   &NiagaraLightRendererProperties
   &NiagaraMeshRendererProperties
  ]
  renderers [
   {rendererClass=&NiagaraSpriteRendererProperties}
   {rendererClass=&NiagaraLightRendererProperties, rendererIndex=1}
   {rendererClass=&NiagaraMeshRendererProperties, rendererIndex=2}
  ]
 }
 FireTrails {
  emitterUpdateScript [
   &EmitterState(Life Cycle Mode=::NewEnumerator1, Loop Behavior=::NewEnumerator1, Loop Delay=0.15, Loop Duration=0.8, UseLoopDelay=true)
   &SpawnRate(
    SpawnRate=&FloatFromCurve(
     FloatCurve=DataInterface() {propertyValues {Curve {keys [{interpMode=RCIM_Cubic, tangentMode=RCTM_User, time=0.3, value=0.5}, {interpMode=RCIM_Cubic, tangentMode=RCTM_User, time=0.6, value=1}]}}}
     CurveIndex=$Emitter.NormalizedLoopAge
     Scale Curve=1200
    )
    Use Spawn Probability=false
   )
  ]
  particleSpawnScript [
   &PartitionParticles(Number Of Partitions=3, Write Partition to Ribbon ID=::NewEnumerator0)
   &InitializeParticle(
    Lifetime=&FloatFromCurve(
     FloatCurve=DataInterface() {propertyValues {Curve {keys [{interpMode=RCIM_Cubic, tangentMode=RCTM_User, time=0.5, value=1}, {interpMode=RCIM_Cubic, tangentMode=RCTM_User, time=1, value=0.2}]}}}
     CurveIndex=$Emitter.Age
     Scale Curve=0.25
    )
    Ribbon Width=18
    Ribbon Width Mode=::NewEnumerator1
    UsePositionOffset=false
    Write Material Random=true
   )
   &SystemLocation(
    MaskSpawnGroup=false
    Offset=&AddVectorToPosition(
     Position=&SelectVectorFromArray(Direct Array Index=$Particles.PartitionParticles.Partition, Vector Selection Array=$System.Position_HammerExhausts)
     Vector=&MakeVector(
      X=0
      Y=0
      Z=&FloatFromCurve(FloatCurve=DataInterface() {propertyValues {Curve {keys [<at 0.5 0>,<at 1 1>]}}}, CurveIndex=$Emitter.Age, Scale Curve=-5)
     )
    )
   ) {enabled=false}
   /SetVariables(Particles.Position=&AddVectorToPosition(
     Position=$Engine.Emitter.SimulationPosition
     Vector=&RotateVectorByQuaternion(
      Vector To Rotate=&AddVectorToPosition(
       Position=&SelectVectorFromArray(Direct Array Index=$Particles.PartitionParticles.Partition, Vector Selection Array=$System.Position_HammerExhausts)
       Vector=&MakeVector(
        X=0
        Y=0
        Z=&FloatFromCurve(
         FloatCurve=DataInterface() {propertyValues {Curve {keys [
             {arriveTangent=0.4222, interpMode=RCIM_Cubic, leaveTangent=0.4222, tangentMode=RCTM_User, time=0.55, value=0}
             {interpMode=RCIM_Cubic, tangentMode=RCTM_User, time=1, value=1}
            ]}}}
         CurveIndex=$Emitter.Age
         Scale Curve=-4
        )
       )
      )
      Quaternion=$Engine.Owner.Rotation
     )
    )) {isSetParameters=true}
  ]
  particleUpdateScript [
   &AccelerationForce(Acceleration=<0 0 1000>) {enabled=false}
   &SolveForcesAndVelocity(Clamp Velocity=false, Limit Acceleration=false)
   &ParticleState()
   &ScaleColor(
    Scale Alpha=&FloatFromCurve(FloatCurve=DataInterface() {propertyValues {Curve {keys [
         {interpMode=RCIM_Cubic, tangentMode=RCTM_User, time=0, value=0}
         {arriveTangent=-0.6565, interpMode=RCIM_Cubic, leaveTangent=-0.6565, tangentMode=RCTM_User, time=0.1, value=1}
         {arriveTangent=-0.9894, interpMode=RCIM_Cubic, leaveTangent=-0.9894, tangentMode=RCTM_User, time=1, value=0}
        ]}}})
    Scale RGB=&MakeVectorFromLinearColorRGB(Color=&ColorFromCurve(ColorCurve=DataInterface() {propertyValues {
        AlphaCurve {keys [<at 0 1 RCIM_Cubic>]}
        BlueCurve {keys [<at 0 0.06 RCIM_Cubic>,<at 0.2419 0 RCIM_Cubic>,<at 0.7493 0.0083>]}
        GreenCurve {keys [<at 0 2.04 RCIM_Cubic>,<at 0.2419 0.0667 RCIM_Cubic>,<at 0.7493 0.006>]}
        RedCurve {keys [<at 0 6 RCIM_Cubic>,<at 0.2419 2 RCIM_Cubic>,<at 0.7493 0.02>]}
       }}))
    ScaleA=true
    ScaleRGB=true
   )
   &ScaleColor(
    Scale Alpha=&FloatFromCurve(FloatCurve=DataInterface() {propertyValues {Curve {keys [<at 0.5 1>,<at 1 0>]}}}, CurveIndex=$Emitter.Age)
    ScaleA=true
    ScaleRGB=true
   ) {enabled=false, module=ScaleColor001}
   &ScaleRibbonWidth(Ribbon Width Scale=&FloatFromCurve(FloatCurve=DataInterface() {propertyValues {Curve {keys [<at 0 1>,<at 1 0.35>]}}}))
   &DynamicMaterialParameters(
    Index 0 Param 1=&FloatFromCurve(FloatCurve=DataInterface() {propertyValues {Curve {keys [
         {arriveTangent=-1.8277, interpMode=RCIM_Cubic, leaveTangent=-1.8277, tangentMode=RCTM_User, time=0, value=1}
         {interpMode=RCIM_Cubic, tangentMode=RCTM_User, time=1, value=0}
        ]}}})
    Index 0 Param 2=2
    Index 0 Param 3=&SelectFloatFromArray(
     Direct Array Index=$Particles.PartitionParticles.Partition
     Float Selection Array=DataInterface() {propertyValues {FloatData [1.2,0.84,0.7]}}
    )
    Module.Index0Param0Write=true
    Module.Index0Param1Write=true
    Module.Index0Param2Write=true
    Module.Index0Param3Write=false
   )
  ]
  rendererClasses [&NiagaraRibbonRendererProperties]
  renderers [{rendererClass=&NiagaraRibbonRendererProperties}]
 }
 Flames {
  emitterUpdateScript [
   &EmitterState()
   &SpawnBurst_Instantaneous(Spawn Count=3, Spawn Time=0.1, Use Spawn Probability=false, UseLoopCountLimit=false)
  ]
  particleSpawnScript [
   &InitializeParticle(
    Lifetime=0.98
    Mesh Scale=&SelectVectorFromArray(
     Direct Array Index=&ReturnExecIndex()
     Vector Selection Array=DataInterface() {propertyValues {FloatData [<0.06 0.06 0.2>, <0.075 0.075 0.2>, <0.06 0.06 0.2>]}}
    )
    Mesh Scale Mode=::NewEnumerator4
    Sprite Rotation Mode=::NewEnumerator2
    Sprite Size Mode=::NewEnumerator3
    Uniform Sprite Size=75
    UsePositionOffset=false
    Write Material Random=true
   )
   &SystemLocation(
    MaskSpawnGroup=false
    Offset=&AddVectorToPosition(
     Position=&SelectVectorFromArray(Direct Array Index=&ReturnExecIndex(), Vector Selection Array=$System.Position_HammerExhausts)
     Vector=<0 0 0>
    )
   ) {enabled=false}
   &InitialMeshOrientation(Mesh Orientation Mode=::NewEnumerator1, Rotation=<0.25 0 0>, Use Rotation Vector=true)
   &CameraOffset(Camera Offset Amount=1)
  ]
  particleUpdateScript [
   /SetVariables(Particles.Position=&AddVectorToPosition(
     Position=&SelectVectorFromArray(Direct Array Index=&ReturnExecIndex(), Vector Selection Array=$System.Position_HammerExhausts)
     Vector=&MakeVector(
      X=0
      Y=&FloatFromCurve(
       FloatCurve=DataInterface() {propertyValues {Curve {keys [{interpMode=RCIM_Cubic, tangentMode=RCTM_User, time=0, value=1}, {interpMode=RCIM_Cubic, tangentMode=RCTM_User, time=1, value=-1}]}}}
       Scale Curve=-2
      )
      Z=0
     )
    )) {isSetParameters=true}
   /SetVariables(Particles.PositionGlow:NiagaraPosition=&AddVectorToPosition(
     Position=$Particles.Position
     Vector=&MakeVector(
      X=0
      Y=&FloatFromCurve(
       FloatCurve=DataInterface() {propertyValues {Curve {keys [{interpMode=RCIM_Cubic, tangentMode=RCTM_User, time=0, value=1}, {interpMode=RCIM_Cubic, tangentMode=RCTM_User, time=1, value=0.5}]}}}
       Scale Curve=-8
      )
      Z=0
     )
    )) {isSetParameters=true}
   &ParticleState()
   &ScaleMeshSize(Scale Factor=&MakeVector(
     X=&FloatFromCurve(FloatCurve=DataInterface() {propertyValues {Curve {keys [
          {interpMode=RCIM_Cubic, tangentMode=RCTM_User, time=0, value=1}
          {interpMode=RCIM_Cubic, tangentMode=RCTM_User, time=0.27, value=0.8}
          {interpMode=RCIM_Cubic, tangentMode=RCTM_User, time=0.3, value=1}
          {arriveTangent=-0.0298, interpMode=RCIM_Cubic, leaveTangent=-0.0298, tangentMode=RCTM_User, time=0.7, value=1}
          {arriveTangent=-1.0476, interpMode=RCIM_Cubic, leaveTangent=-1.0476, tangentMode=RCTM_User, time=1, value=0.75}
         ]}}})
     Y=&FloatFromCurve(FloatCurve=DataInterface() {propertyValues {Curve {keys [
          {interpMode=RCIM_Cubic, tangentMode=RCTM_User, time=0, value=1}
          {interpMode=RCIM_Cubic, tangentMode=RCTM_User, time=0.27, value=0.8}
          {interpMode=RCIM_Cubic, tangentMode=RCTM_User, time=0.3, value=1}
          {arriveTangent=-0.0298, interpMode=RCIM_Cubic, leaveTangent=-0.0298, tangentMode=RCTM_User, time=0.7, value=1}
          {arriveTangent=-1.0476, interpMode=RCIM_Cubic, leaveTangent=-1.0476, tangentMode=RCTM_User, time=1, value=0.75}
         ]}}})
     Z=&FloatFromCurve(FloatCurve=DataInterface() {propertyValues {Curve {keys [
          {interpMode=RCIM_Cubic, tangentMode=RCTM_User, time=0, value=0.3}
          {arriveTangent=3.2251, interpMode=RCIM_Cubic, leaveTangent=3.2251, tangentMode=RCTM_User, time=0.16, value=0.6}
          {interpMode=RCIM_Cubic, tangentMode=RCTM_User, time=0.23, value=1.5}
          {interpMode=RCIM_Cubic, tangentMode=RCTM_User, time=0.4, value=0.7}
          {arriveTangent=-0.0987, interpMode=RCIM_Cubic, leaveTangent=-0.0987, tangentMode=RCTM_User, time=0.77, value=0.8}
          {arriveTangent=-15.0489, interpMode=RCIM_Cubic, leaveTangent=-15.0489, tangentMode=RCTM_User, time=0.82, value=3}
          {arriveTangent=-0.3788, interpMode=RCIM_Cubic, leaveTangent=-0.3788, tangentMode=RCTM_User, time=0.9, value=0.4}
         ]}}})
    ))
   &ScaleColor(
    Scale Alpha=&FloatFromCurve(FloatCurve=DataInterface() {propertyValues {Curve {keys [
         {interpMode=RCIM_Cubic, tangentMode=RCTM_User, time=0.88, value=1}
         {arriveTangent=-0.2092, interpMode=RCIM_Cubic, leaveTangent=-0.2092, tangentMode=RCTM_User, time=1, value=0}
        ]}}})
    Scale RGB=&MakeVectorFromLinearColorRGB(Color=&ColorFromCurve(ColorCurve=DataInterface() {propertyValues {
        AlphaCurve {keys [<at 0 1 RCIM_Cubic>]}
        BlueCurve {keys [<at 0 0.2 RCIM_Cubic>,<at 0.101 0>,<at 0.4313 0>,<at 0.7898 0>,<at 0.8759 0.5>,<at 0.9761 0 RCIM_Cubic>]}
        GreenCurve {keys [<at 0 3.4667 RCIM_Cubic>,<at 0.101 0.1167>,<at 0.4313 0.1167>,<at 0.7898 0.0667>,<at 0.8759 12.875>,<at 0.9761 0.0167 RCIM_Cubic>]}
        RedCurve {keys [<at 0 10 RCIM_Cubic>,<at 0.101 1>,<at 0.4313 1>,<at 0.7898 1>,<at 0.8759 50>,<at 0.9761 1 RCIM_Cubic>]}
       }}))
    ScaleA=true
    ScaleRGB=true
   ) {enabled=false}
   &ScaleColor(
    Scale Alpha=&FloatFromCurve(FloatCurve=DataInterface() {propertyValues {Curve {keys [
         {interpMode=RCIM_Cubic, tangentMode=RCTM_User, time=0.88, value=1}
         {arriveTangent=-0.2092, interpMode=RCIM_Cubic, leaveTangent=-0.2092, tangentMode=RCTM_User, time=1, value=0}
        ]}}})
    Scale RGB=&MakeVectorFromLinearColorRGB(Color=&ColorFromCurve(ColorCurve=DataInterface() {propertyValues {
        AlphaCurve {keys [<at 0 1 RCIM_Cubic>]}
        BlueCurve {keys [
          <at 0.1794 3 RCIM_Cubic>
          <at 0.3404 0.0135>
          {arriveTangent=0.2903, interpMode=RCIM_Cubic, leaveTangent=0.2903, time=0.8058, value=0}
          <at 0.845 0.16>
          <at 0.9037 0>
         ]}
        GreenCurve {keys [
          <at 0.1794 0.5 RCIM_Cubic>
          <at 0.3404 0.676>
          {arriveTangent=0.4994, interpMode=RCIM_Cubic, leaveTangent=0.4994, time=0.8058, value=0.05}
          <at 0.845 0.928>
          <at 0.9037 0.02>
         ]}
        RedCurve {keys [
          <at 0.1794 0 RCIM_Cubic>
          <at 0.3404 3>
          {arriveTangent=1.9818, interpMode=RCIM_Cubic, leaveTangent=1.9818, time=0.8058, value=1}
          <at 0.845 4>
          <at 0.9037 0.4>
         ]}
       }}))
    ScaleA=true
    ScaleRGB=true
   ) {module=ScaleColor001}
   &DynamicMaterialParameters(
    Index 0 Param 1=&RandomRangeFloat(Minimum=0, Maximum=1)
    Index 0 Param 4=&RandomRangeFloat(Minimum=-2, Maximum=-5)
    Module.Index0Param0Write=true
    Module.Index0Param1Write=false
    Module.Index0Param2Write=false
    Module.Index0Param3Write=true
   )
   /SetVariables(Particles.GlowColor=&MakeLinearColorFromVectorAndFloat(
     "Vector (RGB)"=&MakeVectorFromLinearColorRGB(Color=&ColorFromCurve(ColorCurve=DataInterface() {propertyValues {
         AlphaCurve {keys [<at 0 1 RCIM_Cubic>]}
         BlueCurve {keys [
           <at 0.0502 0.5 RCIM_Cubic>
           <at 0.1959 0.15>
           {arriveTangent=-0.1577, interpMode=RCIM_Cubic, leaveTangent=-0.1577, time=0.8157, value=0}
           <at 0.8935 0.04>
           <at 0.9588 0>
          ]}
         GreenCurve {keys [
           <at 0.0502 3.1917 RCIM_Cubic>
           <at 0.1959 0.7675>
           {arriveTangent=-0.5278, interpMode=RCIM_Cubic, leaveTangent=-0.5278, time=0.8157, value=0.1}
           <at 0.8935 0.3993>
           <at 0.9588 0.1>
          ]}
         RedCurve {keys [
           <at 0.0502 10 RCIM_Cubic>
           <at 0.1959 3>
           {arriveTangent=-1.4336, interpMode=RCIM_Cubic, leaveTangent=-1.4336, time=0.8157, value=1}
           <at 0.8935 2>
           <at 0.9588 1>
          ]}
        }}))
     "Float (Alpha)"=&FloatFromCurve(FloatCurve=DataInterface() {propertyValues {Curve {keys [
          <at 0.1 0 RCIM_Cubic>
          {arriveTangent=-0.1014, interpMode=RCIM_Cubic, leaveTangent=-0.1014, tangentMode=RCTM_User, time=0.2, value=1}
          {interpMode=RCIM_Cubic, tangentMode=RCTM_User, time=0.9, value=1}
          {interpMode=RCIM_Cubic, tangentMode=RCTM_User, time=1, value=0}
         ]}}})
    )) {isSetParameters=true}
   &ScaleSpriteSize(Uniform Curve Sprite Scale=DataInterface() {propertyValues {Curve {keys [
        {arriveTangent=0.0929, interpMode=RCIM_Cubic, leaveTangent=0.0929, tangentMode=RCTM_User, time=0, value=0.3}
        {arriveTangent=-4.5934, interpMode=RCIM_Cubic, leaveTangent=-4.5934, tangentMode=RCTM_User, time=0.2, value=1.5}
        {interpMode=RCIM_Cubic, tangentMode=RCTM_User, time=0.4, value=1}
        {arriveTangent=0.0539, interpMode=RCIM_Cubic, leaveTangent=0.0539, tangentMode=RCTM_User, time=0.7, value=1}
        {arriveTangent=-3.5, interpMode=RCIM_Cubic, leaveTangent=-3.5, time=0.77, value=2}
        <at 0.9 0.3 RCIM_Cubic>
       ]}}})
  ]
  rendererClasses [&NiagaraMeshRendererProperties, &NiagaraSpriteRendererProperties]
  renderers [
   {rendererClass=&NiagaraMeshRendererProperties, rendererIndex=0}
   {rendererClass=&NiagaraSpriteRendererProperties, rendererIndex=1}
  ]
 }
 Smoke_Burst {
  emitterUpdateScript [
   &EmitterState(Life Cycle Mode=::NewEnumerator1, Loop Behavior=::NewEnumerator1, Loop Delay=0.017, Loop Duration=0.3, UseLoopDelay=true)
   &SpawnRate(SpawnRate=50, Use Spawn Probability=false)
  ]
  particleSpawnScript [
   &PartitionParticles(Number Of Partitions=3)
   &InitializeParticle(
    Color=<rgba 0.1 0.1 0.1 0.7>
    Lifetime Max=1
    Lifetime Min=0.3
    Lifetime Mode=::NewEnumerator1
    Position Offset=&SelectVectorFromArray(
     Direct Array Index=&SpawnGroup()
     Vector Selection Array=DataInterface() {propertyValues {FloatData [<85 60 25>, <85 -60 25>]}}
    )
    Sprite Size Mode=::NewEnumerator1
    Uniform Sprite Size Max=60
    Uniform Sprite Size Min=40
    UsePositionOffset=false
    Write Material Random=true
   )
   &CameraOffset(Camera Offset Amount=5)
   &SystemLocation(
    MaskSpawnGroup=false
    Offset=&AddVectorToPosition(
     Position=&SelectVectorFromArray(Direct Array Index=$Particles.PartitionParticles.Partition, Vector Selection Array=$System.Position_HammerExhausts)
     Vector=&RandomRangeVector(Minimum=<-4 0 -4>, Maximum=<4 -5 4>)
    )
   ) {enabled=false}
   /SetVariables(Particles.Position=&AddVectorToPosition(
     Position=$Engine.Emitter.SimulationPosition
     Vector=&RotateVectorByQuaternion(
      Vector To Rotate=&AddVectorToPosition(
       Position=&SelectVectorFromArray(Direct Array Index=$Particles.PartitionParticles.Partition, Vector Selection Array=$System.Position_HammerExhausts)
       Vector=&RandomRangeVector(Minimum=<-4 0 -4>, Maximum=<4 -5 4>)
      )
      Quaternion=$Engine.Owner.Rotation
     )
    )) {isSetParameters=true}
   &AddVelocity(
    Module.Use Falloff Distance=false
    Origin Offset=&AddVectorToPosition(
     Position=&SelectVectorFromArray(Direct Array Index=$Particles.PartitionParticles.Partition, Vector Selection Array=$System.Position_HammerExhausts)
     Vector=<0 5 0>
    )
    Override Seed=false
    Velocity Mode=::NewEnumerator1
    Velocity Speed=&RandomRangeFloat(Minimum=150, Maximum=250)
   )
   &InheritVelocity(Inherited Velocity Amount Scale=<0.3 0.3 0.3>, Inherited Velocity Speed Limit=500, Speed Limit=true, UseSpeedThreshold=false) {enabled=false}
  ]
  particleUpdateScript [
   &Drag(Drag=&RandomRangeFloat(Minimum=2, Maximum=4), Use Linear Drag=true, Use Rotational Drag=true)
   &AccelerationForce(Acceleration=<0 0 300>)
   &SolveForcesAndVelocity(Clamp Velocity=false, Limit Acceleration=false)
   &ParticleState()
   &ScaleSpriteSize(
    Uniform Curve Scale=&RandomRangeFloat(Minimum=0.8, Maximum=1.1)
    Uniform Curve Sprite Scale=DataInterface() {propertyValues {Curve {keys [{arriveTangent=1.9788, interpMode=RCIM_Cubic, leaveTangent=1.9788, tangentMode=RCTM_User, time=0, value=0.4}, <at 1 1.2 RCIM_Cubic>]}}}
   )
   &DynamicMaterialParameters(
    Index 0 Param 1=&Add_Float(
     A=&Multiply_Float(A=$Engine.Time, B=0.5)
     B=&RandomRangeFloat(Minimum=0, Maximum=1)
    )
    Index 0 Param 3=&FloatFromCurve(FloatCurve=DataInterface() {propertyValues {Curve {keys [<at 0 0.1>,<at 0.28 0>,<at 0.28 0>]}}})
    Index 0 Param 4=1
    Index 1 Param 1=0.8
    Index 1 Param 2=&RandomRangeFloat(Minimum=0, Maximum=1)
    Index 1 Param 3=1
    Index 1 Param 4=1.5
    Index 2 Param 1=1
    Index 2 Param 2=&RandomRangeFloat(Minimum=0, Maximum=1)
    Index 2 Param 3=&FloatFromCurve(FloatCurve=DataInterface() {propertyValues {Curve {keys [<at 0 0 RCIM_Cubic>, {arriveTangent=2.1666, tangentMode=RCTM_User, time=1, value=1}]}}})
    Index 2 Param 4=1
    Module.Param0WriteEnabled=true
    Module.Param0WriteEnabled1=true
    Module.Param0WriteEnabled2=true
    Module.Param1WriteEnabled=true
    Module.Param1WriteEnabled1=true
    Module.Param1WriteEnabled2=true
    Module.Param2WriteEnabled=true
    Module.Param2WriteEnabled1=true
    Module.Param2WriteEnabled2=true
    Module.Param3WriteEnabled=true
    Module.Param3WriteEnabled1=true
    Module.Param3WriteEnabled2=true
    Write Parameter Index 1=true
    Write Parameter Index 2=true
   )
   &ScaleColor(
    Scale Alpha=&FloatFromCurve(
     FloatCurve=DataInterface() {propertyValues {Curve {keys [
         {interpMode=RCIM_Cubic, tangentMode=RCTM_User, time=0, value=0}
         <at 0.2 0.8 RCIM_Cubic>
         {interpMode=RCIM_Cubic, tangentMode=RCTM_User, time=1, value=0}
        ]}}}
     Scale Curve=&FN_PickFloatDependOnPlatform(TargetPlatformActive=1, TargetPlatformNoActive=1.5)
    )
    Scale RGB=&MakeVectorFromLinearColorRGB(Color=&ColorFromCurve(ColorCurve=DataInterface() {propertyValues {
        AlphaCurve {keys []}
        BlueCurve {keys [<at 0.0233 0.5312 RCIM_Cubic>]}
        GreenCurve {keys [<at 0.0233 0.5312 RCIM_Cubic>]}
        RedCurve {keys [<at 0.0233 0.5312 RCIM_Cubic>]}
       }}))
    ScaleA=true
    ScaleRGB=true
   )
  ]
  rendererClasses [&NiagaraSpriteRendererProperties]
  renderers [
   {rendererClass=&NiagaraSpriteRendererProperties}
   {rendererClass=&NiagaraSpriteRendererProperties, rendererIndex=1}
  ]
 }
 Sparks_Burst {
  emitterUpdateScript [
   &EmitterState(Life Cycle Mode=::NewEnumerator1, Loop Behavior=::NewEnumerator1, Loop Delay=0.03, Loop Duration=0.5, UseLoopDelay=false)
   &SpawnRate(
    SpawnRate=&FloatFromCurve(
     FloatCurve=DataInterface() {propertyValues {Curve {keys [{interpMode=RCIM_Cubic, tangentMode=RCTM_User, time=0, value=1}, {interpMode=RCIM_Cubic, tangentMode=RCTM_User, time=0.6, value=0}]}}}
     CurveIndex=$Emitter.Age
     Scale Curve=100
    )
    Use Spawn Probability=false
   )
   &SpawnBurst_Instantaneous(Spawn Count=20, Use Spawn Probability=false, UseLoopCountLimit=false) {enabled=false}
  ]
  particleSpawnScript [
   &PartitionParticles(Number Of Partitions=3)
   &InitializeParticle(
    Color Maximum=<rgba 1 0.3175 0.09 1>
    Color Minimum=<rgba 1 0.22 0.1 1>
    Color Mode=::NewEnumerator2
    Lifetime=&RandomRangeFloat(Minimum=0.2, Maximum=0.7)
    Sprite Rotation Mode=::NewEnumerator2
    Sprite Size=&Multiply_Vector2D_ByFloat(Float=&RandomRangeFloat(Minimum=0.4, Maximum=3), Vector2D=<1 3>)
    Sprite Size Mode=::NewEnumerator4
    UsePositionOffset=false
    Write Material Random=true
   )
   &CameraOffset(Camera Offset Amount=2)
   &SystemLocation(
    MaskSpawnGroup=false
    Offset=&AddVectorToPosition(
     Position=&SelectVectorFromArray(Direct Array Index=$Particles.PartitionParticles.Partition, Vector Selection Array=$System.Position_HammerExhausts)
     Vector=&RandomRangeVector(Minimum=<-2 0 -2>, Maximum=<2 -10 2>)
    )
   ) {enabled=false}
   /SetVariables(Particles.Position=&AddVectorToPosition(
     Position=$Engine.Emitter.SimulationPosition
     Vector=&RotateVectorByQuaternion(
      Vector To Rotate=&AddVectorToPosition(
       Position=&SelectVectorFromArray(Direct Array Index=$Particles.PartitionParticles.Partition, Vector Selection Array=$System.Position_HammerExhausts)
       Vector=&RandomRangeVector(Minimum=<-2 0 -2>, Maximum=<2 -10 2>)
      )
      Quaternion=$Engine.Owner.Rotation
     )
    )) {isSetParameters=true}
   &ShapeLocation(
    MaskSpawnGroup=false
    Offset=&SelectVectorFromArray(Direct Array Index=$Particles.PartitionParticles.Partition, Vector Selection Array=$System.Position_HammerExhausts)
    Offset Mode=::NewEnumerator0
    Override Seed=false
    Sphere Radius=2
    UseOwnerScale=true
   ) {enabled=false}
   &AddVelocity(
    Module.Use Falloff Distance=false
    Origin Offset=&AddVectorToPosition(
     Position=&SelectVectorFromArray(Direct Array Index=$Particles.PartitionParticles.Partition, Vector Selection Array=$System.Position_HammerExhausts)
     Vector=<0 4 0>
    )
    Override Seed=false
    Velocity Mode=::NewEnumerator1
    Velocity Speed=&RandomRangeFloat(Minimum=130, Maximum=350)
   )
   &InheritVelocity(Inherited Velocity Amount Scale=<0.7 0.7 0.7>, Inherited Velocity Speed Limit=600, Speed Limit=true, UseSpeedThreshold=false)
  ]
  particleUpdateScript [
   &CurlNoiseForce(
    Noise Frequency=60
    Noise Strength=&Multiply_Float(
     A=&FloatFromCurve(FloatCurve=DataInterface() {propertyValues {Curve {keys [<at 0 0>,<at 0.4 1>]}}})
     B=3101.6467
    )
    Pan Noise=true
    Pan Noise Field=<-0.1 0.1 0.1>
   )
   &InheritSourceMovement(
    Applied Position Scale=&VectorFromFloat(Value=&FloatFromCurve(FloatCurve=DataInterface() {propertyValues {Curve {keys [{interpMode=RCIM_Cubic, tangentMode=RCTM_User, time=0, value=1}, {interpMode=RCIM_Cubic, tangentMode=RCTM_User, time=0.2, value=0}]}}}))
    ApplyForce=false
    ApplyPosition=true
    ApplyPositionFalloff=false
   ) {enabled=false}
   &Drag(Drag=4, Use Linear Drag=true, Use Rotational Drag=true)
   &AerodynamicDrag(
    Aerodynamic Drag:NiagaraFloat=&RandomRangeFloat(Minimum=0.05, Maximum=0.3)
    Aerodynamic Rotational Drag=0.1
    Drag Shape Curve=DataInterface() {propertyValues {Curve {keys [
        {interpMode=RCIM_Cubic, tangentMode=RCTM_User, time=0, value=0.1}
        {arriveTangent=2.0799, interpMode=RCIM_Cubic, leaveTangent=2.0799, tangentMode=RCTM_User, time=1, value=1}
       ]}}}
    Lift Contribution=0.2
    Lift Shape Curve=DataInterface() {propertyValues {Curve {keys [
        {interpMode=RCIM_Cubic, tangentMode=RCTM_User, time=0, value=0}
        {interpMode=RCIM_Cubic, tangentMode=RCTM_User, time=0.5, value=1}
        {tangentMode=RCTM_User, time=1, value=0}
       ]}}}
    Pivot Offset Maximum:Vector3f=<1 0.1 0.1>
    Pivot Offset Minimum:Vector3f=<-1 -0.1 -0.1>
    Rotation Strength=0.25
    Use Linear Drag=true
    Use Rotational Drag=true
   )
   &ParticleState()
   &ScaleSpriteSize(Uniform Curve Sprite Scale=DataInterface() {propertyValues {Curve {keys [
        {interpMode=RCIM_Cubic, tangentMode=RCTM_User, time=0.15, value=1}
        {arriveTangent=-1.061, interpMode=RCIM_Cubic, leaveTangent=-1.061, tangentMode=RCTM_User, time=1, value=0}
       ]}}})
   &DynamicMaterialParameters(
    Index 0 Param 1=1
    Module.Index0Param0Write=true
    Module.Index0Param1Write=false
    Module.Index0Param2Write=false
    Module.Index0Param3Write=false
   ) {enabled=false}
   &GravityForce(Gravity=<0 0 160>)
   &SolveForcesAndVelocity(Clamp Velocity=false, Limit Acceleration=false)
   &AlignSpriteToMeshOrientation(SetAlignment=true, SetFacing=true)
   &ScaleColor(
    Scale Alpha=&FloatFromCurve(FloatCurve=DataInterface() {propertyValues {Curve {keys [<at 0 1>,<at 1 0>]}}})
    Scale RGB=&VectorFromFloat(Value=5)
    ScaleA=true
    ScaleRGB=true
   )
  ]
  rendererClasses [&NiagaraSpriteRendererProperties]
  renderers [{rendererClass=&NiagaraSpriteRendererProperties}]
 }
}
summary {
 systemName=NS_CattleJar_Emote_Thrusters
 emitters [
  {emitterName=Flames, rendererClasses [&NiagaraMeshRendererProperties, &NiagaraSpriteRendererProperties]}
  {emitterName=BlackSmokeTrails, rendererClasses [&NiagaraRibbonRendererProperties]}
  {
   emitterName=Engine_GlowLight
   rendererClasses [
    &NiagaraSpriteRendererProperties
    &NiagaraLightRendererProperties
    &NiagaraMeshRendererProperties
   ]
  }
  {emitterName=Sparks_Burst, rendererClasses [&NiagaraSpriteRendererProperties]}
  {emitterName=Smoke_Burst, rendererClasses [&NiagaraSpriteRendererProperties]}
  {emitterName=FireTrails, rendererClasses [&NiagaraRibbonRendererProperties]}
  {bEnabled=false, emitterName=BlackSmokeTrails001, rendererClasses [&NiagaraRibbonRendererProperties]}
 ]
 userVariables [
  User.bIsFrontEnd {type=NiagaraBool}
  User.bIsFrontEndPreview {type=NiagaraBool}
 ]
}
